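# Translucent unlit Niagara material for the smoke of the SplitEnergy vehicle drift trail, with sparkling stars in it.
# Panning smoke times cloud noise, tinted by particle color, drives emissive and, with particle alpha, opacity.
# A brightened, desaturated SpaceTear layer adds stars through panning sphere masks; their sine/cosine pulses always saturate to one.
# Dynamic parameters offset UVs, flip and scale panning, and blend stars from particle color to white.
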Material /VehicleCosmetics/DriftTrail/DriftTrail_SplitEnergy/Materials/M_Drift_Smoke
graph {
 properties {
  bAutomaticallySetUsageInEditor=false
  bUsedWithNiagaraMeshParticles=true
  bUsedWithNiagaraRibbons=true
  bUsedWithNiagaraSprites=true
  blendMode=BLEND_Translucent
  shadingModel=MSM_Unlit
  usedShadingModels=Unlit
 }
 nodes [
  MaterialExpressionVertexColor_0 = ParticleColor()
  n1 = DynamicParameter() {defaultValue=<rgba 1 0 0 1>, paramNames [StarsPaint, TrailOffset, InvertPanner, Speed Multiplier]}
  n2 = Add(A=TextureCoordinate() {uTiling=2}, B=AppendVector(A=@n1.TrailOffset, B=Constant()))
  n3 = Constant() {r=-1}
  n4 = Constant() {r=1}
  n5 = MaterialFunctionCall(X=If(A=@n1.InvertPanner, "A > B"=@n3, "A == B"=@n3, "A < B"=@n4) {constB=1}, Y=@n4) {materialFunction=&MakeFloat2}
  n6 = MaterialFunctionCall(X=@n1.Speed Multiplier, Y=Constant() {r=1}) {materialFunction=&MakeFloat2}
  n7 = Multiply(A=Multiply(A=Constant2Vector() {r=-1.5}, B=@n5.Result), B=@n6.Result)
  n8 = Constant2Vector() {g=0.1, r=-0.7}
  n9 = Panner(Coordinate=Add(A=TextureCoordinate(), B=@n1.TrailOffset) {constB=0.25}, Speed=Multiply(A=@n6.Result, B=Multiply(A=@n5.Result, B=@n8))) {speedX=-0.7, speedY=0.1}
  n10 = Multiply(
   A=TextureSample(UVs=Panner(Coordinate=@n2, Speed=@n7) {speedX=-1.5}) {texture=/DealershipSharedLibrary/FX/Textures/TexturePackage/T_VCVFX_SmokeTrail01_RGB}.R
   B=Power(Base=TextureSample(UVs=@n9) {texture=/DealershipSharedLibrary/FX/Textures/Noise/T_VCVFX_CloudDistort_A}.RGB, Exp=Constant() {r=0.8})
  )
  n11 = DynamicParameter() {defaultValue=<rgba 1 0 0 1>, paramNames [StarsPaint, TrailOffset, InvertPanner, Speed Multiplier]}
  n12 = Add(A=TextureCoordinate() {uTiling=3, vTiling=3}, B=@n11.TrailOffset)
  n13 = Constant() {r=-1}
  n14 = Constant() {r=1}
  n15 = If(A=@n11.InvertPanner, "A > B"=@n13, "A == B"=@n13, "A < B"=@n14) {constB=1}
  n16 = Multiply(
   A=Multiply(
    A=MaterialFunctionCall(X=@n15, Y=@n14) {materialFunction=&MakeFloat2}.Result
    B=Constant2Vector() {r=-1.5}
   )
   B=MaterialFunctionCall(X=@n11.Speed Multiplier, Y=Constant() {r=1}) {materialFunction=&MakeFloat2}.Result
  )
  n17 = Desaturation(None=TextureSample(UVs=Panner(Coordinate=@n12, Speed=@n16) {speedX=-1}) {texture=/DealershipSharedLibrary/FX/Textures/Diffuse/T_VCVFX_SpaceTear_D}.RGB)
  n18 = Constant() {r=15}
  n19 = Time()
  MaterialExpressionConstantClamp_1 = Saturate(None=Add(A=Sine(None=@n19) {period=0.5}) {constB=2})
  n20 = TextureCoordinate() {uTiling=1.5, vTiling=1.5}
  n21 = DynamicParameter() {defaultValue=<rgba 1 0 0 1>, paramNames [StarsPaint, TrailOffset, InvertPanner, Speed Multiplier]}
  n22 = MaterialFunctionCall(X=@n21.Speed Multiplier, Y=Constant() {r=1}) {materialFunction=&MakeFloat2}
  n23 = Constant() {r=-1}
  n24 = Constant() {r=1}
  n25 = MaterialFunctionCall(X=If(A=@n21.InvertPanner, "A > B"=@n23, "A == B"=@n23, "A < B"=@n24) {constB=1}, Y=@n24) {materialFunction=&MakeFloat2}
  n26 = Multiply(A=Constant2Vector() {g=0.05, r=-0.8}, B=@n25.Result)
  n27 = TextureSample(UVs=Panner(Coordinate=Add(A=@n20, B=@n21.TrailOffset), Speed=Multiply(A=@n22.Result, B=@n26)) {speedX=-0.8, speedY=0.05}) {texture=/DealershipSharedLibrary/FX/Textures/Gradient_Masks/T_VCVFX_SphereMaskTimeOffset_A}
  n28 = Multiply(A=@MaterialExpressionConstantClamp_1, B=@n27.R)
  n29 = Time()
  MaterialExpressionConstantClamp_2 = Saturate(None=Add(A=Cosine(None=@n29) {period=0.5}) {constB=2})
  n30 = TextureCoordinate() {uTiling=2, vTiling=2}
  n31 = Multiply(A=@n25.Result, B=Constant2Vector() {g=0.2, r=-0.45})
  n32 = MaterialFunctionCall(X=@n21.Speed Multiplier, Y=Constant() {r=1}) {materialFunction=&MakeFloat2}
  n33 = TextureSample(UVs=Panner(Coordinate=Add(A=@n30, B=@n21.TrailOffset), Speed=Multiply(A=@n31, B=@n32.Result)) {speedX=-0.45, speedY=0.2}) {texture=&T_VCVFX_SphereMaskTimeOffset_A}
  n34 = Multiply(A=@MaterialExpressionConstantClamp_2, B=@n33.R)
  n35 = Multiply(A=@n27.B, B=@MaterialExpressionConstantClamp_1)
  n36 = Multiply(A=@MaterialExpressionConstantClamp_2, B=@n33.B)
  MaterialExpressionConstantClamp_0 = Saturate(None=Add(A=Add(A=@n28, B=@n34), B=Add(A=@n35, B=@n36)))
  n37 = Saturate(None=Multiply(A=Multiply(A=@n17, B=@n18), B=Add(A=@MaterialExpressionConstantClamp_0) {constB=0.1}))
  n38 = Multiply(A=@MaterialExpressionVertexColor_0.RGB, B=@n37)
  n39 = Multiply(A=Constant3Vector() {constant=<rgba 1 1 1 0>}, B=@n37)
  n40 = DynamicParameter() {defaultValue=<rgba 1 0 0 1>, paramNames [StarsPaint, TrailOffset, InvertPanner, Speed Multiplier]}
  MaterialFunctionCall(Emissive=Add(A=Multiply(A=@MaterialExpressionVertexColor_0.RGB, B=@n10), B=LinearInterpolate(A=@n38, B=@n39, Alpha=@n40.StarsPaint))) {materialFunction=&MF_FX_EmissiveAdjustments}
  Saturate(None=Multiply(A=@MaterialExpressionVertexColor_0.A, B=@n10))
 ]
}
